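; Gameplay ability for a battle pass quest that tracks damage dealt shortly after the player exits a giant snowball.
; On activation from a gameplay event, it calls the parent logic, then applies a "Recent Snowball" gameplay effect to the owning pawn's Fort ability system component and ends immediately.
; That effect serves as the timer marking the post-snowball window.
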(event Ability|EventActivateAbilityFromEvent
  (params (EventData :type Gameplay%20Event%20Data%20Structure%20%28by%20ref%29))
  ;[1 1 1] Activate ability - listen for when player exits a giant snowball, then start the timer
  (|CallParentFunction EventData)
  (GameplayEffects|ApplyGameplayEffectToSelf :self (Ability|GetFortAbilitySystemComponent (Variables|GAVars|Data|GetOwningPlayerPawn)) :GameplayEffectClass (Variables|Default|GetRecent%20Snowball) :Level 0.0)
  (Ability|EndAbility))
